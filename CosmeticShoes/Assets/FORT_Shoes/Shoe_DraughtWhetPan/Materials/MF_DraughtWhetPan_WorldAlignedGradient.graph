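# Material function for the DraughtWhetPan shoe that adds an animated red gradient glow to incoming material attributes.
# Despite its WorldAligned name, it works in UV space, feeding scaled, offset UVs through polar coordinates into a panning ramp texture.
# A stepped UV mask (LERPMask) and a TechArtMask input, previewed by the shoe FX texture, supply masking via named reroutes.
# The masked result blends into BaseColor and EmissiveColor, with the ToggleEmissive parameter scaling both blends.
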MaterialFunction /CosmeticShoes/Assets/FORT_Shoes/Shoe_DraughtWhetPan/Materials/MF_DraughtWhetPan_WorldAlignedGradient
graph {nodes [
  n1 = GetMaterialAttributes(None=FunctionInput() {inputName=MA, inputType=FunctionInput_MaterialAttributes}) {attributeGetTypes [#1, #2]}
  n2 = Constant3Vector() {constant=<rgba 0.7483 0 0 0>}
  n3 = TextureCoordinate()
  n4 = Constant2Vector() {g=0.4, r=0.25}
  n5 = Constant2Vector() {g=-0.364, r=0.89}
  n6 = Constant2Vector() {g=0.65, r=-0.9173}
  n7 = NamedRerouteUsage()
  n8 = MaterialFunctionCall!(
   UV Coord=Add(
    A=MaterialFunctionCall(UVs=@n3, Texture Scale=@n4) {materialFunction=&ScaleUVsByCenter}.UVs
    B=LinearInterpolate(A=@n5, B=@n6, Alpha=@n7)
   )
   LenghtScale
   RadialScale
  ) {materialFunction=/Game/Global/MaterialFunctions/TechArt/MF_PolarCoordinates}
  n9 = Multiply(
   A=TextureSample(UVs=Panner(Coordinate=@n8.Result) {speedX=-0.5}) {texture=/Game/Global/Textures/Default/Masks_and_Gradients/T_Ramp_Virt}.R
   B=NamedRerouteUsage()
  )
  n10 = LinearInterpolate(B=@n2, Alpha=@n9) {constA=0.02, constB=0.02}
  n11 = Multiply(A=ScalarParameter() {defaultValue=1, parameterName=ToggleEmissive}, B=NamedRerouteUsage())
  n12 = Multiply(A=@n9, B=@n2)
  FunctionOutput(None=SetMaterialAttributes!(
    MaterialAttributes=@n1.MaterialAttributes
    BaseColor=LinearInterpolate(A=@n1.BaseColor, B=@n10, Alpha=@n11) {constB=0.02}
    EmissiveColor=LinearInterpolate(A=@n1.EmissiveColor, B=@n12, Alpha=@n11)
   ) {attributeSetTypes [#1, #2]})
  n13 = TextureCoordinate()
  n14 = ComponentMask(None=@n13) {r=true}
  n15 = Add(A=Multiply(A=OneMinus(None=@n14)) {constB=2}, B=ComponentMask(None=@n13) {g=true})
  NamedRerouteDeclaration(None=Saturate(None=Step(Y=@n15) {constX=1.4})) {name=LERPMask, nodeColor=<rgba 1 0 0.6094 1>}
  NamedRerouteDeclaration(None=FunctionInput(Preview=TextureSample() {texture=/CosmeticShoes/Assets/FORT_Shoes/Shoe_DraughtWhetPan/Textures/T_DraughtWhetPan_Shoe_FX}.R) {inputName=TechArtMask, inputType=FunctionInput_Scalar}) {name=TechArtMask, nodeColor=<rgba 1 0.0234 0 1>}
 ]}
