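# Parent material for the DraughtWhetPan cosmetic shoe, built on material attributes with the shared 2023 character parent functions.
# The PreFX output passes through shoe reactivity, a shoe-specific world-aligned gradient fed by the tech art mask's red channel, and the sock function.
# PostFX and shoe hologram functions follow.
# A final step clamps specular to a parameter maximum, blended by the cloth fuzz channel of the tech art mask.
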Material /CosmeticShoes/Assets/FORT_Shoes/Shoe_DraughtWhetPan/Materials/M_DraughtWhetPan_Shoe_Parent
graph {
 properties {
  bUseMaterialAttributes=true
  bUsedWithClothing=true
  bUsedWithMeshDeformer=true
  bUsedWithMorphTargets=true
  bUsedWithNanite=true
  bUsedWithNiagaraMeshParticles=true
  bUsedWithSkeletalMesh=true
  pixelDepthOffsetMode=PDOM_Legacy
  shadingModel=MSM_FromMaterialExpression
  usedShadingModels=Default Lit | Subsurface | Clear Coat | Subsurface Profile | Cloth
 }
 nodes [
  n1 = MaterialFunctionCall() {materialFunction=&MF_CharacterParent_2023_PreFX}
  n2 = MaterialFunctionCall!(
   InOpacity
   Material=MaterialFunctionCall!(
    TechArtMask=ComponentMask(None=@n1.Tech Art Mask) {r=true}
    MA=MaterialFunctionCall!(Material=@n1.Result) {materialFunction=/Game/Global/Materials/Cosmetics/Kicks/Shoes/Materials/MF_ShoeReactivityTest}.Result
   ) {materialFunction=&/MF_DraughtWhetPan_WorldAlignedGradient}.Result
   Mask=@n1.Tech Art Mask
  ) {materialFunction=/Game/Global/Materials/Cosmetics/Kicks/Socks/Materials/MF_Sock}
  n3 = GetMaterialAttributes(None=MaterialFunctionCall!(In=MaterialFunctionCall(Material=@n2.Result, M=@n1.M, Pre Skinned Local Position=@n1.M, Distance Blend=@n1.M, Fresnel=@n1.M) {materialFunction=&MF_CharacterParent_2023_PostFX}.Result) {materialFunction=/Game/Global/Materials/Cosmetics/Kicks/Shoes/Materials/MF_TechartShoeHologram}.Result) {attributeGetTypes [#1, #2]}
  n4 = ScalarParameter() {defaultValue=1, group=04_Tech Art, parameterName=Specular Clamp}
  SetMaterialAttributes!(
   MaterialAttributes=@n3.MaterialAttributes
   Specular=LinearInterpolate(
    A=@n3.Specular
    B=Clamp(None=@n3.Specular, Max=@n4)
    Alpha=ChannelMaskParameter(None=@n1.Tech Art Mask) {group=05_Tech Art Masks, parameterName=ClothFuzzMaskChannel}.R
   )
   Anisotropy
  ) {attributeSetTypes [#1, #2]}
 ]
}
